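# Masked, two-sided lit material for Niagara mesh particles, apparently the spawn effect of the male Alien Flora pickaxe.
# It uses the pickaxe's diffuse, normal and adjusted emissive textures.
# Dynamic parameters drive normal-based inflation plus a UV1-masked sine wobble, shaped for vertex offset.
# They also drive a dissolve-style mask combining a swirly-noise threshold with a movable UV1 sphere-mask gradient broken up by panning blurred noise.
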Material /BRCosmetics/Weapons/FORT_Melee/Pickaxe_Alien_Flora_Male/FX/M_Pickaxe_Alien_Flora_Male_SpawnMs
graph {
 properties {
  bUsedWithNiagaraMeshParticles=true
  blendMode=BLEND_Masked
  pixelDepthOffsetMode=PDOM_Legacy
  previewMesh=&/SM_Pickaxe_Alien_Flora_Male_03
  twoSided=true
  usedShadingModels=Default Lit
 }
 nodes [
  n1 = VertexNormalWS()
  n2 = DynamicParameter()
  n3 = ComponentMask(None=TextureCoordinate() {coordinateIndex=1, uTiling=0.4, vTiling=0.5}) {g=true}
  n4 = Constant() {r=2}
  n5 = Sine(None=Add(A=Multiply(A=@n3, B=@n4), B=Time()))
  n6 = TextureCoordinate() {coordinateIndex=1}
  n7 = Constant() {r=0.5}
  n8 = Multiply(
   A=MaterialFunctionCall(X=@n5, Y=@n5, Z=@n5) {materialFunction=&MakeFloat3}.Result
   B=SphereMask(A=ComponentMask(None=@n6) {g=true}, B=@n7, Radius=@n7) {hardnessPercent=0}
  )
  Add(
   A=Multiply(A=Normalize(VectorInput=@n1), B=@n2.Param1)
   B=Multiply(A=@n2.Param4, B=Transform(None=@n8) {transformSourceType=TRANSFORMSOURCE_Local, transformType=TRANSFORM_Tangent})
  )
  MaterialFunctionCall(Emissive=TextureSample() {texture=/BRCosmetics/Weapons/FORT_Melee/Pickaxe_Alien_Flora_Male/Textures/T_Pickaxe_Alien_Flora_Male_E}.RGB) {materialFunction=&MF_FX_EmissiveAdjustments}
  n9 = Panner(Coordinate=TextureCoordinate() {coordinateIndex=1, uTiling=0.5, vTiling=0.5})
  n10 = Constant() {r=1}
  n11 = SphereMask(A=ComponentMask(None=TextureCoordinate() {coordinateIndex=1}) {g=true}, B=@n2.Param3, Radius=Constant() {r=2}) {attenuationRadius=1, hardnessPercent=0}
  n12 = TextureSample(UVs=Panner(Coordinate=TextureCoordinate() {coordinateIndex=1, uTiling=0.5, vTiling=0.3}) {speedY=0.2}) {samplerType=SAMPLERTYPE_Grayscale, texture=&LowResBlurredNoise}
  Multiply(
   A=If(
    A=TextureSample(UVs=@n9) {texture=/Game/Weapons/WeaponSkins/Textures/T_Swirly_Noise_highContrast}.R
    B=DynamicParameter() {parameterIndex=1, defaultValue=<rgba 0 1 1 1>}.Param1
    "A > B"=@n10
    "A == B"=@n10
    "A < B"=Constant()
   )
   B=LinearInterpolate(A=@n11, B=Multiply(A=@n11, B=@n12.R), Alpha=@n2.Param2) {constAlpha=0.7}
  )
  ScalarParameter() {defaultValue=-1}
  TextureSample() {texture=/BRCosmetics/Weapons/FORT_Melee/Pickaxe_Alien_Flora_Male/Textures/T_Pickaxe_Alien_Flora_Male_D}
  TextureSample() {samplerType=SAMPLERTYPE_Normal, texture=/BRCosmetics/Weapons/FORT_Melee/Pickaxe_Alien_Flora_Male/Textures/T_Pickaxe_Alien_Flora_Male_N}
 ]
}
